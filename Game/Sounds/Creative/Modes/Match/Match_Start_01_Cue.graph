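# Sound cue for the match start sound in Creative mode.
# A Delay node with a fixed half-second wait feeds a single Wave Player for the Match_Start_01 wave.
# It uses the Creative match attenuation asset with a very large max distance and routes through the critical in-game UI sound class.
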SoundCue /Game/Sounds/Creative/Modes/Match/Match_Start_01_Cue
root=Delay_0
nodes [
 {id=Delay_0, children [WavePlayer_0], properties {delayMax=0.5, delayMin=0.5}}
 {id=WavePlayer_0, properties {soundWaveAssetPtr=&/Match_Start_01}}
]
properties {
 attenuationSettings=/Game/Sounds/Creative/Attenuations/Fort_Creative_Match_Att
 maxDistance=2097152
 soundClassObject=&Fort_UI_InGame_Critical
 volumeMultiplier=1
}
