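# Sound cue that plays the match stop sound for Creative mode matches.
# A single wave player feeds the output, routed through the critical in-game UI sound class with Creative match attenuation, a raised priority and slightly reduced volume.
# A shared match-end concurrency set governs playback, alongside a one-instance override that prevents new copies from starting.
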SoundCue /Game/Sounds/Creative/Modes/Match/Match_Stop_01_Cue
root=WavePlayer_0
nodes [{id=WavePlayer_0, properties {soundWaveAssetPtr=&/Match_Stop_01}}]
properties {
 attenuationSettings=/Game/Sounds/Creative/Attenuations/Fort_Creative_Match_Att
 maxDistance=2097152
 priority=0.8
 soundClassObject=&Fort_UI_InGame_Critical
 volumeMultiplier=0.8
 concurrencyOverrides {maxCount=1, resolutionRule=PreventNew}
 concurrencySet [/Game/Sounds/Creative/Concurrency/Creative_MatchEnd_Conc]
}
